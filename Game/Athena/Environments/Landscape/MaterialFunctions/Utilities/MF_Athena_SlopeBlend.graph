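# Material function for the Athena landscape that outputs slope-based blend alphas broken up by a height input.
# The absolute vertex normal Z, scaled by Transition Multiplier and clamped, serves as the transition phase for HeightLerp, with the scaled Height input as height texture.
# It outputs contrasted and uncontrasted alphas.
# A second HeightLerp adds ColormapTransitionBias from MPC_Landscape to the phase, producing a separate alpha for the colormap texture.
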MaterialFunction /Game/Athena/Environments/Landscape/MaterialFunctions/Utilities/MF_Athena_SlopeBlend
graph {nodes [
  n1 = ComponentMask(None=VertexNormalWS()) {b=true}
  n2 = Constant() {r=0.975}
  n3 = Clamp(None=Multiply(
    A=Abs(None=@n1)
    B=FunctionInput(Preview=@n2) {bUsePreviewValueAsDefault=true, inputName=Transition Multiplier, inputType=FunctionInput_Scalar, sortPriority=1}
   ))
  n4 = Multiply(
   A=FunctionInput() {inputName=Height, inputType=FunctionInput_Scalar}
   B=FunctionInput(Preview=Constant() {r=0.1}) {bUsePreviewValueAsDefault=true, inputName=Height Multiplier, inputType=FunctionInput_Scalar, sortPriority=2}
  )
  n5 = Constant() {r=5}
  n6 = MaterialFunctionCall(Transition Phase=@n3, Height Texture=@n4, Contrast=@n5) {materialFunction=&HeightLerp}
  FunctionOutput(None=@n6.Alpha) {outputName=Alpha}
  FunctionOutput(None=@n6.Lerp Alpha No Contrast) {outputName=Alpha_NoContrast, sortPriority=1}
  n7 = MaterialFunctionCall(
   Transition Phase=Add(A=CollectionParameter() {collection=/Game/Athena/Environments/Landscape/MPC/MPC_Landscape, parameterName=ColormapTransitionBias}, B=@n3)
   Height Texture=@n4
   Contrast=@n5
  ) {
   desc=Height lerp for ColorMap Texture - has ColormapTransitionBias
   materialFunction=&HeightLerp
  }
  FunctionOutput(None=@n7.Alpha) {outputName=Alpha_ColorMap, sortPriority=2}
 ]}
