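# Material function that converts world position into 0–1 terrain UVs spanning the Apollo landscape.
# The default output remaps XY with fixed offset and size constants and saturates; WithInterpolator passes this through a vertex interpolator.
# Adjustable exposes offset, subtract and divide through the LandscapeWorldUV parameter.
# MatchForMinimap instead takes mesh texture coordinates, scales them slightly about center and rotates a quarter turn.
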MaterialFunction /Game/Athena/Environments/Landscape/MaterialFunctions/Utilities/MF_Apollo_Terrain_Coords
graph {nodes [
  n1 = FunctionInput(Preview=WorldPosition().XYZ) {bUsePreviewValueAsDefault=true, inputName=WorldPosition}
  n2 = Divide(A=Subtract(A=ComponentMask(None=@n1) {g=true, r=true}, B=Constant() {r=-130048}), B=Constant() {r=260096})
  FunctionOutput(None=Saturate(None=@n2)) {sortPriority=-10}
  FunctionOutput(None=Saturate(None=VertexInterpolator(VS=@n2).PS)) {outputName=WithInterpolator, sortPriority=1}
  n3 = TextureCoordinate()
  n4 = Constant() {r=0.96}
  FunctionOutput(None=MaterialFunctionCall(
    UVs=MaterialFunctionCall(UVs=@n3, Texture Scale=@n4) {materialFunction=&ScaleUVsByCenter}.UVs
    "Rotation Angle (0-1)"=Constant() {r=0.25}
   ) {materialFunction=&CustomRotator}.Rotated Values) {outputName=MatchForMinimap, sortPriority=3}
  n5 = VectorParameter() {
   group=BiomeSettings
   parameterName=LandscapeWorldUV
   channelNames {a=LWUV-Divide, b=LWUV-Subtract, g=LWUV-Y, r=LWUV-X}
   defaultValue=<rgba 0 0 -130048 260096>
  }
  n6 = Subtract(A=ComponentMask(None=Add(A=@n1, B=@n5)) {g=true, r=true}, B=@n5.LWUV-Subtract)
  FunctionOutput(None=Saturate(None=Divide(A=@n6, B=@n5.LWUV-Divide))) {outputName=Adjustable, sortPriority=2}
 ]}
