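# Material function that converts world position into normalized 0-1 UV coordinates spanning the Athena terrain.
# It offsets the XY of the WorldPosition input, which defaults to absolute world position, by a fixed map origin, divides by a constant map size and saturates.
# A second output, WithInterpolator, computes the coordinates per vertex through a VertexInterpolator before saturating.
# A half-extent branch feeds neither output.
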MaterialFunction /Game/Athena/Environments/Landscape/MaterialFunctions/Utilities/MF_Athena_Terrain_Coords
graph {nodes [
  n1 = Constant() {r=306200.7812}
  n2 = Divide(A=@n1)
  n3 = Multiply(A=@n2) {constB=-1}
  AppendVector(A=@n3, B=@n3)
  n4 = FunctionInput(Preview=WorldPosition().XYZ) {bUsePreviewValueAsDefault=true, inputName=WorldPosition}
  n5 = Divide(A=Subtract(A=ComponentMask(None=@n4) {g=true, r=true}, B=Constant() {r=-153000}), B=@n1)
  FunctionOutput(None=Saturate(None=@n5))
  FunctionOutput(None=Saturate(None=VertexInterpolator(VS=@n5).PS)) {outputName=WithInterpolator}
  Multiply(A=@n2)
 ]}
